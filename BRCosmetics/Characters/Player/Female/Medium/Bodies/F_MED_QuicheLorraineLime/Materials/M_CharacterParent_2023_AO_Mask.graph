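# Parent character material variant that adds optional ambient occlusion from a mask, for skeletal, cloth and other mesh uses.
# It wraps the shared CharacterParent 2023 PreFX and PostFX material functions, editing material attributes between them.
# A selectable channel of the PreFX Tech Art Mask, blue by default, is inverted and written to Ambient Occlusion.
# The UseAOMask static switch chooses between that result and the untouched PreFX attributes before PostFX.
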Material /BRCosmetics/Characters/Player/Female/Medium/Bodies/F_MED_QuicheLorraineLime/Materials/M_CharacterParent_2023_AO_Mask
graph {
 properties {
  bUseMaterialAttributes=true
  bUsedWithClothing=true
  bUsedWithMeshDeformer=true
  bUsedWithMorphTargets=true
  bUsedWithNanite=true
  bUsedWithNiagaraMeshParticles=true
  bUsedWithSkeletalMesh=true
  pixelDepthOffsetMode=PDOM_Legacy
  shadingModel=MSM_FromMaterialExpression
  usedShadingModels=Default Lit | Subsurface | Clear Coat | Subsurface Profile | Cloth
 }
 nodes [
  n1 = MaterialFunctionCall() {materialFunction=&MF_CharacterParent_2023_PreFX}
  n2 = GetMaterialAttributes(None=@n1.Result) {attributeGetTypes [#1]}
  n3 = LinearInterpolate(Alpha=ChannelMaskParameter(None=@n1.Tech Art Mask) {group=AOMask, maskChannel=Blue, parameterName=AOChannel, defaultValue=<rgba 0 0 1 0>}.R) {constA=1, constB=0}
  MaterialFunctionCall(
   Material=StaticSwitchParameter(
    True=SetMaterialAttributes!(MaterialAttributes=@n2.MaterialAttributes, Roughness=@n2.MaterialAttributes, Ambient Occlusion=@n3) {attributeSetTypes [#1, #2]}
    False=@n1.Result
   ) {group=AOMask, parameterName=UseAOMask}
   M=@n1.M
   Pre Skinned Local Position=@n1.M
   Distance Blend=@n1.M
   Fresnel=@n1.M
  ) {materialFunction=&MF_CharacterParent_2023_PostFX}
 ]
}
