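# Animation Blueprint for the F_MED_QuicheLorraineLime player body, copying its pose from the attached parent mesh and adjusting it for this character.
# The copied pose runs through the character's Control Rig, then additive bone-space tweaks offset and roll attach_backpack and offset attach_cape.
# Each update sets ShortHairEnable by wildcard-matching the pawn's Face part mesh name against this character's FaceAcc, though no visible graph node reads that flag.
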AnimBlueprint /BRCosmetics/Characters/Player/Female/Medium/Bodies/F_MED_QuicheLorraineLime/Meshes/F_MED_QuicheLorraineLime_AnimBP
target_skeleton=&/F_MED_QuicheLorraineLime_Skeleton
graphs [
 AnimGraph {nodes [
   n1 = Control Rig(Source=Copy Pose From Mesh() {bUseAttachedParent=true}) {controlRigAssetReference {blueprintRigClass=&/F_MED_QuicheLorraineLime_CtrlRig.F_MED_QuicheLorraineLime_CtrlRig_C}}
   n2 = "Transform (Modify) Bone - Bone: attach_backpack"!(ComponentPose=Local To Component(LocalPose=@n1), bAlphaBoolEnabled:bool=True, Translation="0,5,0", Rotation="0,0,8", Scale="1,1,1") -> (Pose) {
    alphaInputType=Bool
    rotationMode=BMM_Additive
    rotationSpace=BCS_BoneSpace
    scaleMode=BMM_Additive
    scaleSpace=BCS_BoneSpace
    translationMode=BMM_Additive
    translationSpace=BCS_BoneSpace
    alphaBoolBlend {blendOption=Sinusoidal}
    boneToModify {boneName=attach_backpack}
    rotation {roll=8}
    translation {y=5}
   }
   n3 = "Transform (Modify) Bone - Bone: attach_cape"!(ComponentPose=@n2, bAlphaBoolEnabled:bool=True, Translation="0,2.5,0", Rotation="0,0,0", Scale="1,1,1") -> (Pose) {
    alphaInputType=Bool
    rotationMode=BMM_Additive
    rotationSpace=BCS_BoneSpace
    scaleMode=BMM_Additive
    scaleSpace=BCS_BoneSpace
    translationMode=BMM_Additive
    translationSpace=BCS_BoneSpace
    alphaBoolBlend {blendOption=Sinusoidal}
    boneToModify {boneName=attach_cape}
    translation {y=2.5}
   }
   Output Pose(Result=Component To Local(ComponentPose=@n3)) {layerGroup=None}
  ]}
 EventGraph {nodes [
   Event BlueprintUpdateAnimation()
   > n1 = Cast To FortPlayerPawn(Object=TryGetPawnOwner())
   n2 = GetSkeletalMeshForPartType(self=@n1, PartType=Face)
   > Set ShortHairEnable(ShortHairEnable:bool=MatchesWildcard(SourceString=GetDisplayName(Object=@n2), Wildcard=*F_MED_QuicheLorraineLime_FaceAcc*, SearchCase=CaseSensitive)) -> (Output_Get:bool)
  ]}
]
variables [ShortHairEnable {type=boolean}]
