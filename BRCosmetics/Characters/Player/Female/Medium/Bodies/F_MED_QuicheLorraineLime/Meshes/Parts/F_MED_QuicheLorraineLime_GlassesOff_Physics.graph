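# Physics asset for the GlassesOff mesh part of the QuicheLorraineLime female medium character, covering the dynamic bones of the hair bangs.
# Capsule bodies form branching center, left and right strand chains rooted at the head, several sharing top bones.
# Constraints limit nearly every joint to 45-degree swing and 1-degree twist, with one tighter swing on left strand C.
# Unconstrained clavicle and neck capsules, plus the head capsule, provide body shapes around the hair.
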PhysicsAsset /BRCosmetics/Characters/Player/Female/Medium/Bodies/F_MED_QuicheLorraineLime/Meshes/Parts/F_MED_QuicheLorraineLime_GlassesOff_Physics
bodies {
 clavicle_l [{shapeType=Capsule, length=8, radius=6, shapeName=clavicle_l_capsule, center=<5.6788 -1.2175 -0.3265>, rotation=<rot -80 179.9996 -179.9998>}]
 clavicle_r [{shapeType=Capsule, length=8, radius=6, shapeName=clavicle_r_capsule, center=<-5.6821 0 0.3083>, rotation=<rot 80 0.0001 -0.0002>}]
 dyn_bangs_c_A_01 [{shapeType=Capsule, length=1, radius=2.5, shapeName=dyn_bangs_c_A_01_capsule, center=<1.3253 0 0.2234>, rotation=<rot 90 0 0>}]
 dyn_bangs_c_A_02 [{shapeType=Capsule, length=0.5, radius=3, shapeName=dyn_bangs_c_A_02_capsule, center=<1.3253 0 0.2234>, rotation=<rot 90 0.0001 0>}]
 dyn_bangs_c_A_03 [{shapeType=Capsule, length=0.5, radius=3, shapeName=dyn_bangs_c_A_03_capsule, center=<2.1365 0 0.1695>, rotation=<rot 90 0.0002 0>}]
 dyn_bangs_c_B_01 [{shapeType=Capsule, length=1, radius=2.5, shapeName=dyn_bangs_c_B_01_capsule, center=<1.3253 0 0.2234>, rotation=<rot 90 0.0002 0>}]
 dyn_bangs_c_B_02 [{shapeType=Capsule, length=1, radius=2.5, shapeName=dyn_bangs_c_B_02_capsule, center=<1.3253 0 0.2234>, rotation=<rot 90 0.0002 0>}]
 dyn_bangs_c_B_03 [{shapeType=Capsule, length=1, radius=2.5, shapeName=dyn_bangs_c_B_03_capsule, center=<1.3253 0 0.2234>, rotation=<rot 90 0.0003 0>}]
 dyn_bangs_c_top [{shapeType=Capsule, length=3, radius=0.5, shapeName=dyn_bangs_c_top_capsule, center=<1.3253 0 0.2234>, rotation=<rot 90 0.0001 0>}]
 dyn_bangs_l_A_01 [{shapeType=Capsule, length=2, radius=1.5, shapeName=dyn_bangs_l_A_01_capsule, center=<1.3253 0 0.2234>, rotation=<rot 90 0.0005 0>}]
 dyn_bangs_l_A_02 [{shapeType=Capsule, length=2, radius=1.5, shapeName=dyn_bangs_l_A_02_capsule, center=<1.8674 0 0.2234>, rotation=<rot 90 0.0004 0>}]
 dyn_bangs_l_A_03 [{shapeType=Capsule, length=1.5, radius=2, shapeName=dyn_bangs_l_A_03_capsule, center=<1.8674 0 0.2234>, rotation=<rot 90 0.0003 0>}]
 dyn_bangs_l_BC_top [{shapeType=Capsule, length=2, radius=1.5, shapeName=dyn_bangs_l_BC_top_capsule, center=<1.3253 0 0.2234>, rotation=<rot 90 0.0004 0>}]
 dyn_bangs_l_B_01 [{shapeType=Capsule, length=2, radius=1.5, shapeName=dyn_bangs_l_B_01_capsule, center=<1.3253 0 0.2234>, rotation=<rot 90 0.0003 0>}]
 dyn_bangs_l_B_02 [{shapeType=Capsule, length=2.5, radius=2, shapeName=dyn_bangs_l_B_02_capsule, center=<1.3253 0 0.2234>, rotation=<rot 90 0.0002 0>}]
 dyn_bangs_l_C_01 [{shapeType=Capsule, length=2, radius=2, shapeName=dyn_bangs_l_C_01_capsule, center=<1.3253 0 0.2234>, rotation=<rot 90 0.0004 0>}]
 dyn_bangs_l_C_02 [{shapeType=Capsule, length=2.5, radius=1, shapeName=dyn_bangs_l_C_02_capsule, center=<1.3253 0 0.2234>, rotation=<rot 90 0.0003 0>}]
 dyn_bangs_l_D_01 [{shapeType=Capsule, length=2.5, radius=1.5, shapeName=dyn_bangs_l_D_01_capsule, center=<1.3253 0 0.2234>, rotation=<rot 90 0.0004 0>}]
 dyn_bangs_l_D_02 [{shapeType=Capsule, length=2.5, radius=1.5, shapeName=dyn_bangs_l_D_02_capsule, center=<1.3253 0 0.2234>, rotation=<rot 90 0.0002 0>}]
 dyn_bangs_l_D_03 [{shapeType=Capsule, length=2.5, radius=1, shapeName=dyn_bangs_l_D_03_capsule, center=<1.3253 0 0.2234>, rotation=<rot 90 0.0003 0>}]
 dyn_bangs_l_D_04 [{shapeType=Capsule, length=2.5, radius=1, shapeName=dyn_bangs_l_D_04_capsule, center=<1.1181 0.5694 0.2234>, rotation=<rot 90 20.0004 0>}]
 dyn_bangs_r_AB_top [{shapeType=Capsule, length=2, radius=1.5, shapeName=dyn_bangs_r_AB_top_capsule, center=<1.3253 0 0.2234>, rotation=<rot 90 -0.0004 0>}]
 dyn_bangs_r_A_01 [{shapeType=Capsule, length=1.5, radius=2, shapeName=dyn_bangs_r_A_01_capsule, center=<1.3253 0 0.2234>, rotation=<rot 90 -0.0004 0>}]
 dyn_bangs_r_A_02 [{shapeType=Capsule, length=2.5, radius=1, shapeName=dyn_bangs_r_A_02_capsule, center=<1.3253 0 0.2234>, rotation=<rot 90 -0.0003 0>}]
 dyn_bangs_r_B_01 [{shapeType=Capsule, length=1.5, radius=2, shapeName=dyn_bangs_r_B_01_capsule, center=<1.3253 0 0.2234>, rotation=<rot 90 -0.0004 0>}]
 dyn_bangs_r_B_02 [{shapeType=Capsule, length=2.5, radius=1, shapeName=dyn_bangs_r_B_02_capsule, center=<1.3253 0 0.2234>, rotation=<rot 90 -0.0002 0>}]
 dyn_bangs_r_B_03 [{shapeType=Capsule, length=2, radius=1.5, shapeName=dyn_bangs_r_B_03_capsule, center=<1.6736 0 0.2234>, rotation=<rot 90 -0.0001 0>}]
 dyn_bangs_r_B_04 [{
   shapeType=Capsule
   length=4.5
   radius=1.5
   shapeName=dyn_bangs_r_B_04_capsule
   center=<2.2474 0.2228 1.4031>
   rotation=<rot 80 179.9998 -179.9999>
  }]
 dyn_bangs_r_C_01 [{shapeType=Capsule, length=2.5, radius=1, shapeName=dyn_bangs_r_C_01_capsule, center=<1.3727 0 -0.3368>, rotation=<rot 60 -0.0003 0>}]
 dyn_bangs_r_D_01 [{shapeType=Capsule, length=2.5, radius=1, shapeName=dyn_bangs_r_D_01_capsule, center=<1.3253 0 0.2234>, rotation=<rot 90 -0.0003 0>}]
 dyn_bangs_r_D_02 [{shapeType=Capsule, length=2, radius=1.5, shapeName=dyn_bangs_r_D_02_capsule, center=<1.3253 0 0.2234>, rotation=<rot 90 -0.0004 0>}]
 dyn_bangs_r_D_03 [{shapeType=Capsule, length=2, radius=1.5, shapeName=dyn_bangs_r_D_03_capsule, center=<1.3253 0 0.2234>, rotation=<rot 90 -0.0004 0>}]
 dyn_bangs_r_D_04 [{shapeType=Capsule, length=2.5, radius=1, shapeName=dyn_bangs_r_D_04_capsule, center=<1.3253 0 0.2234>, rotation=<rot 90 -0.0003 0>}]
 dyn_bangs_r_top [{shapeType=Capsule, length=3, radius=0.5, shapeName=dyn_bangs_r_top_capsule, center=<1.3253 0 0.2234>, rotation=<rot 90 -0.0005 0>}]
 head [{shapeType=Capsule, length=5, radius=6, shapeName=head_capsule, center=<0.798 2.6223 0>, rotation=<rot 90 -0.0002 0>}]
 neck_02 [{shapeType=Capsule, length=4.5, radius=4.5, shapeName=neck_02_capsule, center=<-0.8007 0.1412 0>, rotation=<rot 90 -10.0002 0>}]
}
constraints [
 {bone1Name=dyn_bangs_c_top, bone2Name=head, swing1LimitDegrees=45, swing2LimitDegrees=45, twistLimitDegrees=1}
 {bone1Name=dyn_bangs_c_A_01, bone2Name=dyn_bangs_c_top, swing1LimitDegrees=45, swing2LimitDegrees=45, twistLimitDegrees=1}
 {bone1Name=dyn_bangs_c_A_02, bone2Name=dyn_bangs_c_A_01, swing1LimitDegrees=45, swing2LimitDegrees=45, twistLimitDegrees=1}
 {bone1Name=dyn_bangs_c_A_03, bone2Name=dyn_bangs_c_A_02, swing1LimitDegrees=45, swing2LimitDegrees=45, twistLimitDegrees=1}
 {bone1Name=dyn_bangs_c_B_01, bone2Name=dyn_bangs_c_top, swing1LimitDegrees=45, swing2LimitDegrees=45, twistLimitDegrees=1}
 {bone1Name=dyn_bangs_c_B_02, bone2Name=dyn_bangs_c_B_01, swing1LimitDegrees=45, swing2LimitDegrees=45, twistLimitDegrees=1}
 {bone1Name=dyn_bangs_c_B_03, bone2Name=dyn_bangs_c_B_02, swing1LimitDegrees=45, swing2LimitDegrees=45, twistLimitDegrees=1}
 {bone1Name=dyn_bangs_l_A_01, bone2Name=head, swing1LimitDegrees=45, swing2LimitDegrees=45, twistLimitDegrees=1}
 {bone1Name=dyn_bangs_l_A_02, bone2Name=dyn_bangs_l_A_01, swing1LimitDegrees=45, swing2LimitDegrees=45, twistLimitDegrees=1}
 {bone1Name=dyn_bangs_l_A_03, bone2Name=dyn_bangs_l_A_02, swing1LimitDegrees=45, swing2LimitDegrees=45, twistLimitDegrees=1}
 {bone1Name=dyn_bangs_l_BC_top, bone2Name=head, swing1LimitDegrees=45, swing2LimitDegrees=45, twistLimitDegrees=1}
 {bone1Name=dyn_bangs_l_B_01, bone2Name=dyn_bangs_l_BC_top, swing1LimitDegrees=45, swing2LimitDegrees=45, twistLimitDegrees=1}
 {bone1Name=dyn_bangs_l_B_02, bone2Name=dyn_bangs_l_B_01, swing1LimitDegrees=45, swing2LimitDegrees=45, twistLimitDegrees=1}
 {bone1Name=dyn_bangs_l_C_01, bone2Name=dyn_bangs_l_BC_top, swing1LimitDegrees=25, swing2LimitDegrees=45, twistLimitDegrees=1}
 {bone1Name=dyn_bangs_l_C_02, bone2Name=dyn_bangs_l_C_01, swing1LimitDegrees=45, swing2LimitDegrees=45, twistLimitDegrees=1}
 {bone1Name=dyn_bangs_l_D_01, bone2Name=head, swing1LimitDegrees=45, swing2LimitDegrees=45, twistLimitDegrees=1}
 {bone1Name=dyn_bangs_l_D_02, bone2Name=dyn_bangs_l_D_01, swing1LimitDegrees=45, swing2LimitDegrees=45, twistLimitDegrees=1}
 {bone1Name=dyn_bangs_l_D_03, bone2Name=dyn_bangs_l_D_02, swing1LimitDegrees=45, swing2LimitDegrees=45, twistLimitDegrees=1}
 {bone1Name=dyn_bangs_l_D_04, bone2Name=dyn_bangs_l_D_03, swing1LimitDegrees=45, swing2LimitDegrees=45, twistLimitDegrees=1}
 {bone1Name=dyn_bangs_r_top, bone2Name=head, swing1LimitDegrees=45, swing2LimitDegrees=45, twistLimitDegrees=1}
 {bone1Name=dyn_bangs_r_AB_top, bone2Name=dyn_bangs_r_top, swing1LimitDegrees=45, swing2LimitDegrees=45, twistLimitDegrees=1}
 {bone1Name=dyn_bangs_r_A_01, bone2Name=dyn_bangs_r_AB_top, swing1LimitDegrees=45, swing2LimitDegrees=45, twistLimitDegrees=1}
 {bone1Name=dyn_bangs_r_A_02, bone2Name=dyn_bangs_r_A_01, swing1LimitDegrees=45, swing2LimitDegrees=45, twistLimitDegrees=1}
 {bone1Name=dyn_bangs_r_B_01, bone2Name=dyn_bangs_r_AB_top, swing1LimitDegrees=45, swing2LimitDegrees=45, twistLimitDegrees=1}
 {bone1Name=dyn_bangs_r_B_02, bone2Name=dyn_bangs_r_B_01, swing1LimitDegrees=45, swing2LimitDegrees=45, twistLimitDegrees=1}
 {bone1Name=dyn_bangs_r_B_03, bone2Name=dyn_bangs_r_B_02, swing1LimitDegrees=45, swing2LimitDegrees=45, twistLimitDegrees=1}
 {bone1Name=dyn_bangs_r_B_04, bone2Name=dyn_bangs_r_B_03, swing1LimitDegrees=45, swing2LimitDegrees=45, twistLimitDegrees=1}
 {bone1Name=dyn_bangs_r_C_01, bone2Name=dyn_bangs_r_B_02, swing1LimitDegrees=45, swing2LimitDegrees=45, twistLimitDegrees=1}
 {bone1Name=dyn_bangs_r_D_01, bone2Name=dyn_bangs_r_top, swing1LimitDegrees=45, swing2LimitDegrees=45, twistLimitDegrees=1}
 {bone1Name=dyn_bangs_r_D_02, bone2Name=dyn_bangs_r_D_01, swing1LimitDegrees=45, swing2LimitDegrees=45, twistLimitDegrees=1}
 {bone1Name=dyn_bangs_r_D_03, bone2Name=dyn_bangs_r_D_02, swing1LimitDegrees=45, swing2LimitDegrees=45, twistLimitDegrees=1}
 {bone1Name=dyn_bangs_r_D_04, bone2Name=dyn_bangs_r_D_03, swing1LimitDegrees=45, swing2LimitDegrees=45, twistLimitDegrees=1}
]
